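# Physics asset for the one-handed Flower Skeleton Male pickaxe, giving its decorative roses, vine and thorns limited secondary motion.
# Capsule bodies cover the handle and each decorative bone, with angular-limited constraints tying them back to the Melee_TwoHanded_Handle root body.
# Roses hang individually off the handle, while the vine and five thorn groups form short chains.
# Thorns get tighter swing limits than roses and vine, so they move more stiffly.
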PhysicsAsset /BRCosmetics/Weapons/FORT_Melee/Pickaxe_Flower_Skeleton_Male_1h/Meshes/Flower_Skeleton_Male_1h_Physics
bodies {
 Melee_TwoHanded_Handle [{shapeType=Capsule, length=4.4594, radius=3.3937, center=<-0.1773 -2.8174 -7.1378>, rotation=<rot 9.8404 -90 1.5106>}]
 dyn_rose_01 [{shapeType=Capsule, length=1.8641, radius=0.8062, center=<-0.2672 -0.648 0.773>, rotation=<rot -41.9045 -120.6133 15.7051>}]
 dyn_rose_010 [{shapeType=Capsule, length=2.5737, radius=0.8685, center=<-0.1395 0.046 0.0548>, rotation=<rot -45.9239 94.3762 59.8528>}]
 dyn_rose_011 [{shapeType=Capsule, length=1.5157, radius=1.118, center=<0.2934 0.9288 0.3001>, rotation=<rot 7.4461 164.776 -72.726>}]
 dyn_rose_012 [{shapeType=Capsule, length=2.0173, radius=0.7584, center=<-0.5097 0.8125 -0.3365>, rotation=<rot 67.8194 91.4768 151.2701>}]
 dyn_rose_013 [{shapeType=Capsule, length=1.255, radius=0.505, center=<0.2343 -1.0724 -0.1297>, rotation=<rot 41.6974 6.2834 -99.0387>}]
 dyn_rose_014 [{shapeType=Capsule, length=1.9263, radius=0.898, center=<-0.3649 -0.5016 0.3805>, rotation=<rot -21.5239 -50.0305 -55.8031>}]
 dyn_rose_015 [{shapeType=Capsule, length=1.51, radius=0.75, center=<-0.1182 -0.9641 -0.3113>, rotation=<rot -62.8959 134.2639 132.041>}]
 dyn_rose_02 [{shapeType=Capsule, length=2.6709, radius=0.7233, center=<-1.0107 -0.1237 0.3049>, rotation=<rot -26.1148 136.95 54.2517>}]
 dyn_rose_03 [{shapeType=Capsule, length=2.3177, radius=0.8656, center=<0.5045 0.8184 0.0737>, rotation=<rot 8.9298 -2.466 -89.9996>}]
 dyn_rose_04 [{shapeType=Capsule, length=1.8276, radius=0.7001, center=<0.4091 0.7712 -0.0708>, rotation=<rot -46.2164 -39.094 -78.7078>}]
 dyn_rose_05 [{shapeType=Capsule, length=2.25, radius=0.75, center=<1.2434 -0.108 0.1122>, rotation=<rot -43.9009 -89.9996 90.7866>}]
 dyn_rose_06 [{shapeType=Capsule, length=1.8644, radius=0.7938, center=<0.3799 -1.1774 0.219>, rotation=<rot -43.8243 -113.1194 45.2507>}]
 dyn_rose_07 [{shapeType=Capsule, length=1.76, radius=0.75, center=<-0.0961 -1.1017 -0.0732>, rotation=<rot -60.5589 1.7577 -97.7258>}]
 dyn_rose_08 [{shapeType=Capsule, length=1.5944, radius=0.5946, center=<-0.4221 -0.8879 0.0933>, rotation=<rot -80.8153 -61.3515 -50.0035>}]
 dyn_rose_09 [{shapeType=Capsule, length=2.51, radius=0.75, center=<0.0451 -1.0717 0.5685>, rotation=<rot -60.1313 -64.9333 -29.8976>}]
 dyn_vine_01 [{shapeType=Capsule, length=1.9505, radius=0.75, center=<2.6806 -0.3394 0.2542>, rotation=<rot 9.7935 90.5616 -87.9463>}]
 dyn_vine_02 [{shapeType=Capsule, length=1.9176, radius=0.6284, center=<1.6197 -0.042 0.3223>, rotation=<rot 0 81.3427 -92.9958>}]
 dyn_vine_03 [{shapeType=Capsule, length=1.9365, radius=0.6566, center=<1.6308 -0.0133 0.5324>, rotation=<rot 0.6155 89.089 -91.6916>}]
 dyn_vine_04 [{shapeType=Capsule, length=1.6538, radius=0.747, center=<1.7695 0.0789 0.3082>, rotation=<rot 0 62.4871 -60.9937>}]
 thorn_a_01 [{shapeType=Capsule, length=0.7974, radius=0.6815, center=<0.4836 0.1153 -0.008>, rotation=<rot 0 -6.454 -87.47>}]
 thorn_a_02 [{shapeType=Capsule, length=2, radius=1, center=<0.492 -0.0344 -0.0553>, rotation=<rot -4.0601 -90.7588 -89.9996>}]
 thorn_a_03 [{shapeType=Capsule, length=2.01, radius=0.75, center=<0.7939 -0.0502 -0.0031>, rotation=<rot 0 84.7722 -88.4297>}]
 thorn_b_01 [{shapeType=Capsule, length=1.1606, radius=0.7144, rotation=<rot -90 -178.7341 178.7342>}]
 thorn_b_02 [{shapeType=Capsule, length=1.755, radius=0.505, center=<1.5214 0 0>, rotation=<rot -90 -2.5305 2.5305>}]
 thorn_c_01 [{shapeType=Capsule, length=1.8346, radius=0.6606, rotation=<rot -90 -2.5305 2.5305>}]
 thorn_c_02 [{shapeType=Capsule, length=1.755, radius=0.505, center=<1.2302 0 0.0001>, rotation=<rot -90 1.8985 -1.8984>}]
 thorn_d_01 [{shapeType=Capsule, length=1.5396, radius=0.7969, rotation=<rot -90 2.5305 -2.5304>}]
 thorn_d_02 [{shapeType=Capsule, length=2.005, radius=0.505, center=<1.1891 0 0.0001>, rotation=<rot -90 174.9489 -174.9488>}]
 thorn_e_01 [{shapeType=Capsule, length=1.8563, radius=0.7203, rotation=<rot -90 1.8985 -1.8984>}]
 thorn_e_02 [{shapeType=Capsule, length=1.755, radius=0.505, center=<0.9463 0 0>, rotation=<rot -90 -178.7341 178.7342>}]
}
constraints [
 {bone1Name=dyn_rose_01, bone2Name=Melee_TwoHanded_Handle, swing1LimitDegrees=20, swing2LimitDegrees=20, twistLimitDegrees=10}
 {bone1Name=dyn_rose_02, bone2Name=Melee_TwoHanded_Handle, swing1LimitDegrees=20, swing2LimitDegrees=20, twistLimitDegrees=10}
 {bone1Name=dyn_rose_03, bone2Name=Melee_TwoHanded_Handle, swing1LimitDegrees=20, swing2LimitDegrees=20, twistLimitDegrees=10}
 {bone1Name=dyn_rose_04, bone2Name=Melee_TwoHanded_Handle, swing1LimitDegrees=20, swing2LimitDegrees=20, twistLimitDegrees=10}
 {bone1Name=dyn_rose_05, bone2Name=Melee_TwoHanded_Handle, swing1LimitDegrees=20, swing2LimitDegrees=20, twistLimitDegrees=10}
 {bone1Name=dyn_rose_06, bone2Name=Melee_TwoHanded_Handle, swing1LimitDegrees=20, swing2LimitDegrees=20, twistLimitDegrees=10}
 {bone1Name=dyn_rose_07, bone2Name=Melee_TwoHanded_Handle, swing1LimitDegrees=20, swing2LimitDegrees=20, twistLimitDegrees=10}
 {bone1Name=dyn_rose_08, bone2Name=Melee_TwoHanded_Handle, swing1LimitDegrees=20, swing2LimitDegrees=20, twistLimitDegrees=10}
 {bone1Name=dyn_rose_09, bone2Name=Melee_TwoHanded_Handle, swing1LimitDegrees=20, swing2LimitDegrees=20, twistLimitDegrees=10}
 {bone1Name=dyn_rose_010, bone2Name=Melee_TwoHanded_Handle, swing1LimitDegrees=20, swing2LimitDegrees=20, twistLimitDegrees=10}
 {bone1Name=dyn_rose_011, bone2Name=Melee_TwoHanded_Handle, swing1LimitDegrees=20, swing2LimitDegrees=20, twistLimitDegrees=10}
 {bone1Name=dyn_rose_012, bone2Name=Melee_TwoHanded_Handle, swing1LimitDegrees=20, swing2LimitDegrees=20, twistLimitDegrees=10}
 {bone1Name=dyn_rose_013, bone2Name=Melee_TwoHanded_Handle, swing1LimitDegrees=20, swing2LimitDegrees=20, twistLimitDegrees=10}
 {bone1Name=dyn_rose_014, bone2Name=Melee_TwoHanded_Handle, swing1LimitDegrees=20, swing2LimitDegrees=20, twistLimitDegrees=10}
 {bone1Name=dyn_rose_015, bone2Name=Melee_TwoHanded_Handle, swing1LimitDegrees=20, swing2LimitDegrees=20, twistLimitDegrees=10}
 {bone1Name=dyn_vine_01, bone2Name=Melee_TwoHanded_Handle, swing1LimitDegrees=20, swing2LimitDegrees=25, twistLimitDegrees=5}
 {bone1Name=dyn_vine_02, bone2Name=dyn_vine_01, swing1LimitDegrees=20, swing2LimitDegrees=25, twistLimitDegrees=5}
 {bone1Name=dyn_vine_03, bone2Name=dyn_vine_02, swing1LimitDegrees=20, swing2LimitDegrees=25, twistLimitDegrees=5}
 {bone1Name=dyn_vine_04, bone2Name=dyn_vine_03, swing1LimitDegrees=20, swing2LimitDegrees=25, twistLimitDegrees=5}
 {bone1Name=thorn_b_01, bone2Name=Melee_TwoHanded_Handle, swing1LimitDegrees=10, swing2LimitDegrees=10, twistLimitDegrees=5}
 {bone1Name=thorn_b_02, bone2Name=thorn_b_01, swing1LimitDegrees=10, swing2LimitDegrees=10, twistLimitDegrees=5}
 {bone1Name=thorn_c_01, bone2Name=Melee_TwoHanded_Handle, swing1LimitDegrees=10, swing2LimitDegrees=10, twistLimitDegrees=5}
 {bone1Name=thorn_c_02, bone2Name=thorn_c_01, swing1LimitDegrees=10, swing2LimitDegrees=10, twistLimitDegrees=5}
 {bone1Name=thorn_d_01, bone2Name=Melee_TwoHanded_Handle, swing1LimitDegrees=10, swing2LimitDegrees=10, twistLimitDegrees=5}
 {bone1Name=thorn_d_02, bone2Name=thorn_d_01, swing1LimitDegrees=10, swing2LimitDegrees=10, twistLimitDegrees=5}
 {bone1Name=thorn_e_01, bone2Name=Melee_TwoHanded_Handle, swing1LimitDegrees=10, swing2LimitDegrees=10, twistLimitDegrees=5}
 {bone1Name=thorn_e_02, bone2Name=thorn_e_01, swing1LimitDegrees=10, swing2LimitDegrees=10, twistLimitDegrees=5}
 {bone1Name=thorn_a_01, bone2Name=Melee_TwoHanded_Handle, swing1LimitDegrees=10, swing2LimitDegrees=3, twistLimitDegrees=5}
 {bone1Name=thorn_a_02, bone2Name=thorn_a_01, swing1LimitDegrees=10, swing2LimitDegrees=3, twistLimitDegrees=5}
 {bone1Name=thorn_a_03, bone2Name=thorn_a_02, swing1LimitDegrees=10, swing2LimitDegrees=3, twistLimitDegrees=5}
]
